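; Building actor blueprint for an Atlantis doorway wall variant with decorative flora.
; On BeginPlay, after the parent logic, it checks whether cosmetic events can run, then sets the flora static mesh component hidden in game on Nintendo Switch or mobile.
; The construction script is empty.
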(event EventBeginPlay
  (|CallParentFunction)
  (Networking|CanExecuteCosmeticEvents
    (:True
      (Development|SetHiddeninGame (Variables|Default|GetSM_Atlantis_Doorway_A_01_Flora_01) (or (TouchPlatform|IsNintendoSwitch) (TouchPlatform|IsMobileGame)) false))))

(fn ConstructionScript)
