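; Building actor Blueprint for an Atlantis quarter wall variant with a flora static mesh component.
; On BeginPlay it calls the parent, then, where cosmetic events can execute, sets that flora component hidden in game on Nintendo Switch or mobile and visible on other platforms.
; The hidden state does not propagate to child components, and the construction script is empty.
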(event EventBeginPlay
  (|CallParentFunction)
  (Networking|CanExecuteCosmeticEvents
    (:True
      (Development|SetHiddeninGame (Variables|Default|GetSM_Atlantis_QuarterWall_A_06_Flora_01) (or (TouchPlatform|IsNintendoSwitch) (TouchPlatform|IsMobileGame)) false))))

(fn ConstructionScript)
